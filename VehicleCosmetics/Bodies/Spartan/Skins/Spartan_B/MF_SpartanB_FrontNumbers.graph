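# Material function for the Spartan B vehicle body skin that draws glowing red front numbers with a blinking part.
# It samples the numbers mask texture, keeping the R channel always lit and switching G on and off with a square wave built from Time, Sine, Ceil and Saturate.
# The Max of both channels drives the opacity mask and, tinted and boosted, the emissive color, output as material attributes.
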MaterialFunction /VehicleCosmetics/Bodies/Spartan/Skins/Spartan_B/MF_SpartanB_FrontNumbers
graph {
 properties {previewBlendMode=BLEND_Masked}
 nodes [
  n1 = TextureSample() {samplerType=SAMPLERTYPE_LinearColor, texture=&/T_SpartanNumbers_M}
  n2 = Multiply(B=Time()) {constA=0.7}
  n3 = Saturate(None=Ceil(None=Sine(None=@n2)))
  n4 = Max(A=@n1.R, B=Multiply(A=@n1.G, B=LinearInterpolate(Alpha=@n3)))
  n5 = Multiply(A=Constant3Vector() {constant=<rgba 1 0.0234 0 0>}, B=@n4)
  FunctionOutput(None=SetMaterialAttributes!(MaterialAttributes, EmissiveColor=Multiply(B=@n5) {constA=3}, OpacityMask=@n4) {attributeSetTypes [#1, #2]})
 ]
}
